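# Physics asset for the FluidWedge medium male player body, covering only its attached gear: pistol, medical item and ammo pieces.
# Each item pairs a tiny box on an anchor bone with an item-sized box on a matching dyn_ bone.
# Nine constraints join each dyn_ bone to its anchor with tight angular limits, which suggests slight secondary sway, with the pistol given the widest range.
# No torso, limb or head bodies are present.
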PhysicsAsset /BRCosmetics/Characters/Player/Male/Medium/Bodies/M_MED_FluidWedge/Meshes/M_MED_FluidWedge_Physics
bodies {
 AmmoBack [{shapeType=Box, extentX=1.01, extentY=1.01, extentZ=1.01, shapeName=AmmoBack_box, rotation=<rot 0 -90 0>}]
 AmmoPelvis_back [{shapeType=Box, extentX=1.01, extentY=1.01, extentZ=1.01, shapeName=AmmoBack_box, rotation=<rot 0 -90 0>}]
 AmmoPelvis_l [{shapeType=Box, extentX=1.01, extentY=1.01, extentZ=1.01, shapeName=AmmoBack_box, rotation=<rot 0 -90 0>}]
 Ammo_l [{shapeType=Box, extentX=1.01, extentY=1.01, extentZ=1.01, shapeName=Ammo_l_box, rotation=<rot 0 -90 0>}]
 Ammo_r [{shapeType=Box, extentX=1.01, extentY=1.01, extentZ=1.01, shapeName=Ammo_r_box, rotation=<rot 0 -90 0>}]
 Pistol [{shapeType=Box, extentX=1.01, extentY=1.01, extentZ=1.01, shapeName=Pistol_box, rotation=<rot 0 -90 0>}]
 dyn_AmmoBack [{
   shapeType=Box
   extentX=5.9719
   extentY=3.9493
   extentZ=9.5993
   shapeName=dyn_AmmoBack_box
   center=<0.7484 -0.2887 -0.8623>
   rotation=<rot 2.7269 -90.1069 -2.7088>
  }]
 dyn_AmmoPelvis_back [{
   shapeType=Box
   extentX=5.9719
   extentY=3.9493
   extentZ=18.5384
   shapeName=dyn_AmmoBack_box
   center=<0.9606 -0.1567 1.9172>
   rotation=<rot 2.7269 -90.1069 4.3552>
  }]
 dyn_AmmoPelvis_l [{
   shapeType=Box
   extentX=5.9719
   extentY=3.9493
   extentZ=18.5384
   shapeName=dyn_AmmoBack_box
   center=<0.9606 -0.1567 1.9172>
   rotation=<rot 2.7269 -90.1069 4.3552>
  }]
 dyn_Ammo_l [{
   shapeType=Box
   extentX=6.64
   extentY=4.4799
   extentZ=19.0422
   shapeName=dyn_Ammo_l_box
   center=<0.1232 -0.0644 -1.3248>
   rotation=<rot 0.5026 -90 1.6778>
  }]
 dyn_Ammo_r [{
   shapeType=Box
   extentX=6.64
   extentY=4.4799
   extentZ=19.0422
   shapeName=dyn_Ammo_r_box
   center=<-0.1233 0.0643 1.3252>
   rotation=<rot 0.5026 -90 1.6778>
  }]
 dyn_Pistol [{
   shapeType=Box
   extentX=7.7548
   extentY=3.4386
   extentZ=20.3963
   shapeName=dyn_Pistol_box
   center=<3.1348 0.1758 -0.3372>
   rotation=<rot 0.2036 -91.8083 -89.2658>
  }]
 dyn_medice [{
   shapeType=Box
   extentX=5.5626
   extentY=10.7684
   extentZ=17.137
   shapeName=dyn_medice_box
   center=<0.3471 0.1111 -1.0379>
   rotation=<rot 0 178.4141 -89.5288>
  }]
 dyn_pistolAmmoA [{
   shapeType=Box
   extentX=4.3925
   extentY=4.99
   extentZ=13.5743
   shapeName=dyn_pistolAmmoA_box
   center=<-0.1603 -0.2791 0.0027>
   rotation=<rot -3.2299 -90 1.0885>
  }]
 dyn_pistolAmmoB [{
   shapeType=Box
   extentX=4.4545
   extentY=4.9948
   extentZ=13.6919
   shapeName=dyn_pistolAmmoB_box
   center=<-0.4995 -0.1798 0.1112>
   rotation=<rot -1.9541 -90 1.6065>
  }]
 medice [{shapeType=Box, extentX=1.01, extentY=1.01, extentZ=1.01, shapeName=medice_box, rotation=<rot 0 -90 0>}]
 pistolAmmoA [{shapeType=Box, extentX=1.01, extentY=1.01, extentZ=1.01, shapeName=AmmoBack_box, rotation=<rot 0 -90 0>}]
 pistolAmmoB [{shapeType=Box, extentX=1.01, extentY=1.01, extentZ=1.01, shapeName=AmmoBack_box, rotation=<rot 0 -90 0>}]
}
constraints [
 {bone1Name=dyn_Pistol, bone2Name=Pistol, swing1LimitDegrees=15, swing2LimitDegrees=10, twistLimitDegrees=10}
 {bone1Name=dyn_medice, bone2Name=medice, swing1LimitDegrees=2, swing2LimitDegrees=10, twistLimitDegrees=5}
 {bone1Name=dyn_AmmoBack, bone2Name=AmmoBack, swing1LimitDegrees=2, swing2LimitDegrees=10, twistLimitDegrees=5}
 {bone1Name=dyn_AmmoPelvis_back, bone2Name=AmmoPelvis_back, swing1LimitDegrees=2, swing2LimitDegrees=10, twistLimitDegrees=5}
 {bone1Name=dyn_AmmoPelvis_l, bone2Name=AmmoPelvis_l, swing1LimitDegrees=2, swing2LimitDegrees=10, twistLimitDegrees=5}
 {bone1Name=dyn_pistolAmmoA, bone2Name=pistolAmmoA, swing1LimitDegrees=2, swing2LimitDegrees=10, twistLimitDegrees=5}
 {bone1Name=dyn_pistolAmmoB, bone2Name=pistolAmmoB, swing1LimitDegrees=2, swing2LimitDegrees=10, twistLimitDegrees=5}
 {bone1Name=dyn_Ammo_l, bone2Name=Ammo_l, swing1LimitDegrees=2, swing2LimitDegrees=2, twistLimitDegrees=10}
 {bone1Name=dyn_Ammo_r, bone2Name=Ammo_r, swing1LimitDegrees=2, swing2LimitDegrees=2, twistLimitDegrees=10}
]
