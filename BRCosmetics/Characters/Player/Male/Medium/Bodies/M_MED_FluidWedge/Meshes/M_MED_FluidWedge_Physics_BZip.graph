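# Physics asset for the FluidWedge medium male player body cosmetic, covering its left and right spine fastener bones.
# Each side pairs a box body on the fastener bone with a smaller box on its dyn_ bone, joined by one constraint.
# Both swing axes on each constraint are free, so the dyn_ pieces can pivot about their fasteners despite the stored swing limit angles.
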PhysicsAsset /BRCosmetics/Characters/Player/Male/Medium/Bodies/M_MED_FluidWedge/Meshes/M_MED_FluidWedge_Physics_BZip
bodies {
 Spine_fastener_l [{shapeType=Box, extentX=4.1748, extentY=5.987, extentZ=1.01, shapeName=Spine_fastener_l_box, center=<0 0 0.6962>, rotation=<rot 0 -90 0>}]
 Spine_fastener_r [{shapeType=Box, extentX=3.3113, extentY=5.1235, extentZ=0.5, shapeName=Spine_fastener_r_box, center=<0 0 -0.5866>, rotation=<rot 0 -90 0>}]
 dyn_Spine_fastener_l [{
   shapeType=Box
   extentX=1.01
   extentY=0.5
   extentZ=2.9321
   shapeName=dyn_Spine_fastener_r_box
   center=<-1.1891 0.0153 -0.1664>
   rotation=<rot 0 -90.095 -90.774>
  }]
 dyn_Spine_fastener_r [{
   shapeType=Box
   extentX=1.01
   extentY=0.5
   extentZ=2.6253
   shapeName=dyn_Spine_fastener_r_box
   center=<0.9831 -0.0149 0.2207>
   rotation=<rot 0 -90.095 -90.774>
  }]
}
constraints [
 {
  bone1Name=dyn_Spine_fastener_l
  bone2Name=Spine_fastener_l
  swing1LimitDegrees=45
  swing1Motion=Free
  swing2LimitDegrees=45
  swing2Motion=Free
  twistLimitDegrees=5
 }
 {
  bone1Name=dyn_Spine_fastener_r
  bone2Name=Spine_fastener_r
  swing1LimitDegrees=45
  swing1Motion=Free
  swing2LimitDegrees=45
  swing2Motion=Free
  twistLimitDegrees=5
 }
]
