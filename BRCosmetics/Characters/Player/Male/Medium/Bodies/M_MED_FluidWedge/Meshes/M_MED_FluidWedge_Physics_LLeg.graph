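# Physics asset for the left-leg fasteners on the M_MED_FluidWedge male player cosmetic body.
# Small dyn_ bodies (capsules at thigh and calf, a box at the foot) are each constrained to a larger box on the matching fastener bone.
# Each joint sets a tight twist limit; thigh swing is free, while calf and foot carry swing limits, wider at the foot.
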PhysicsAsset /BRCosmetics/Characters/Player/Male/Medium/Bodies/M_MED_FluidWedge/Meshes/M_MED_FluidWedge_Physics_LLeg
bodies {
 dyn_fastener_calf_l [{
   shapeType=Capsule
   length=1.8347
   radius=0.5
   shapeName=dyn_fastener_thigh_l_capsule
   center=<-1.9341 0 -0.1757>
   rotation=<rot 89.7162 179.9994 179.9994>
  }]
 dyn_fastener_foot_l [{
   shapeType=Box
   extentX=1.01
   extentY=0.5
   extentZ=2.6253
   shapeName=dyn_Spine_fastener_l_box
   center=<-1.6921 0.0161 -0.2303>
   rotation=<rot 0 -90.095 89.226>
  }]
 dyn_fastener_thigh_l [{shapeType=Capsule, length=1.8347, radius=0.5, shapeName=dyn_fastener_thigh_l_capsule, center=<-2.807 0 -0.1103>, rotation=<rot 85.7102 0 0>}]
 fastener_calf_l [{
   shapeType=Box
   extentX=5.1478
   extentY=7.6775
   extentZ=2.3612
   shapeName=Spine_fastener_l_box
   center=<0.5027 0 1.373>
   rotation=<rot 0 -90 -147.4107>
  }]
 fastener_foot_l [{
   shapeType=Box
   extentX=3.9812
   extentY=7.3514
   extentZ=0.5
   shapeName=Spine_fastener_l_box
   center=<-0.5798 0 0.8127>
   rotation=<rot 0 -90 -176.2797>
  }]
 fastener_thigh_l [{
   shapeType=Box
   extentX=7.7684
   extentY=10.2248
   extentZ=2.9388
   shapeName=Spine_fastener_l_box
   center=<0.4218 0 1.8862>
   rotation=<rot 0 -90 -147.4107>
  }]
}
constraints [
 {bone1Name=dyn_fastener_calf_l, bone2Name=fastener_calf_l, swing1LimitDegrees=45, swing2LimitDegrees=45, twistLimitDegrees=5}
 {bone1Name=dyn_fastener_foot_l, bone2Name=fastener_foot_l, swing1LimitDegrees=90, swing2LimitDegrees=90, twistLimitDegrees=5}
 {
  bone1Name=dyn_fastener_thigh_l
  bone2Name=fastener_thigh_l
  swing1LimitDegrees=45
  swing1Motion=Free
  swing2LimitDegrees=45
  swing2Motion=Free
  twistLimitDegrees=5
 }
]
