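# Physics asset for the right leg of the medium male FluidWedge player body cosmetic, simulating fasteners at the thigh, calf and foot.
# Each location pairs a small box body on a dyn_ fastener bone with a broader, flatter box on its base fastener bone, joined by a constraint.
# All three joints share a tight twist limit; calf and foot carry wide swing limits, while the thigh sets both swing axes free.
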PhysicsAsset /BRCosmetics/Characters/Player/Male/Medium/Bodies/M_MED_FluidWedge/Meshes/M_MED_FluidWedge_Physics_RLeg
bodies {
 dyn_fastener_calf_r [{
   shapeType=Box
   extentX=1.01
   extentY=0.5
   extentZ=2.6253
   shapeName=dyn_Spine_fastener_r_box
   center=<1.6921 -0.0161 0.2303>
   rotation=<rot 0 -90.095 -90.774>
  }]
 dyn_fastener_foot_r [{
   shapeType=Box
   extentX=1.01
   extentY=0.8475
   extentZ=2.6253
   shapeName=dyn_Spine_fastener_r_box
   center=<1.6915 -0.0161 0.2752>
   rotation=<rot 0 -90.095 -90.774>
  }]
 dyn_fastener_thigh_r [{
   shapeType=Box
   extentX=1.01
   extentY=0.5
   extentZ=2.6253
   shapeName=dyn_Spine_fastener_r_box
   center=<1.6921 -0.0161 0.2303>
   rotation=<rot 0 -90.095 -90.774>
  }]
 fastener_calf_r [{
   shapeType=Box
   extentX=5.906
   extentY=6.0245
   extentZ=1.3505
   shapeName=Spine_fastener_r_box
   center=<0.0757 0 -0.4604>
   rotation=<rot 0 -90 29.2799>
  }]
 fastener_foot_r [{
   shapeType=Box
   extentX=5.0149
   extentY=5.8716
   extentZ=1.1606
   shapeName=Spine_fastener_r_box
   center=<0.5792 0 -0.8207>
   rotation=<rot 0 -90 3.7203>
  }]
 fastener_thigh_r [{
   shapeType=Box
   extentX=4.9798
   extentY=6.408
   extentZ=0.7587
   shapeName=Spine_fastener_r_box
   center=<0.1464 0 -0.5884>
   rotation=<rot 0 -90 31.3792>
  }]
}
constraints [
 {bone1Name=dyn_fastener_calf_r, bone2Name=fastener_calf_r, swing1LimitDegrees=90, swing2LimitDegrees=90, twistLimitDegrees=5}
 {bone1Name=dyn_fastener_foot_r, bone2Name=fastener_foot_r, swing1LimitDegrees=90, swing2LimitDegrees=90, twistLimitDegrees=5}
 {
  bone1Name=dyn_fastener_thigh_r
  bone2Name=fastener_thigh_r
  swing1LimitDegrees=45
  swing1Motion=Free
  swing2LimitDegrees=45
  swing2Motion=Free
  twistLimitDegrees=5
 }
]
