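# Post-copy Anim Blueprint for the FluidWedge medium male player body, adding secondary physics motion to the pose copied from its parent mesh.
# The copied pose passes through a Control Rig into a cache feeding two pelvis-space RigidBody sims, switched by bIsInFrontEnd, with softened physics in the front end.
# Chained BZip and per-leg sims follow, then translation constraints pin dynamic leg fastener bones to their anchors.
# The Ammo-to-spine constraints are disconnected from the output.
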AnimBlueprint /BRCosmetics/Characters/Player/Male/Medium/Bodies/M_MED_FluidWedge/Meshes/M_MED_FluidWedge_AnimBP
target_skeleton=&/M_MED_FluidWedge_Skeleton
graphs [
 AnimGraph {nodes [
   n1 = "Constraint - Ammo_l (T)"(Alpha=1, ConstraintWeights_0=1) {boneToModify {boneName=Ammo_l}, constraintSetup [{transformType=Translation, targetBone {boneName=spine_02}}], constraintWeights [1]}
   "Constraint - Ammo_r (T)"(ComponentPose=@n1, Alpha=1, ConstraintWeights_0=1) {boneToModify {boneName=Ammo_r}, constraintSetup [{transformType=Translation, targetBone {boneName=spine_02}}], constraintWeights [1]}
   n2 = RigidBody(ComponentPose=Local To Component(LocalPose=Use cached pose 'PreSim'())) {
    overridePhysicsAsset=&/M_MED_FluidWedge_Physics
    simulationSpace=BaseBoneSpace
    baseBoneRef {boneName=pelvis}
    componentAppliedLinearAccClamp=<300 300 300>
    simSpaceSettings {maxAngularAcceleration=10, maxAngularVelocity=10, maxLinearAcceleration=5, maxLinearVelocity=10, worldAlpha=1}
   }
   n3 = Component To Local(ComponentPose=@n2)
   n4 = RigidBody(ComponentPose=Local To Component(LocalPose=Use cached pose 'PreSim'()), Alpha=0.7) {
    overridePhysicsAsset=&M_MED_FluidWedge_Physics
    simulationSpace=BaseBoneSpace
    baseBoneRef {boneName=pelvis}
    componentAppliedLinearAccClamp=<300 300 300>
    simSpaceSettings {maxAngularAcceleration=10, maxAngularVelocity=10, maxLinearAcceleration=5, maxLinearVelocity=10, worldAlpha=0.4}
   }
   n5 = Component To Local(ComponentPose=@n4)
   n6 = Get bIsInFrontEnd()
   n7 = RigidBody(ComponentPose=Local To Component(LocalPose=Blend Poses by bool(BlendPose_0=@n5, BlendPose_1=@n3, bActiveValue=@n6))) {
    bClampLinearTranslationLimitToRefPose=true
    overridePhysicsAsset=&/M_MED_FluidWedge_Physics_BZip
    simulationSpace=BaseBoneSpace
    baseBoneRef {boneName=pelvis}
    componentAppliedLinearAccClamp=<1500 1500 1500>
    simSpaceSettings {maxAngularAcceleration=50, maxAngularVelocity=50, maxLinearAcceleration=25, maxLinearVelocity=50, worldAlpha=1}
   }
   n8 = RigidBody(ComponentPose=@n7) {
    bClampLinearTranslationLimitToRefPose=true
    bOverrideWorldGravity=true
    overridePhysicsAsset=&/M_MED_FluidWedge_Physics_LLeg
    simulationSpace=BaseBoneSpace
    baseBoneRef {boneName=thigh_l}
    componentAppliedLinearAccClamp=<250 250 250>
    overrideWorldGravity {z=-980}
    simSpaceSettings {maxAngularAcceleration=150, maxAngularVelocity=150, maxLinearAcceleration=2, maxLinearVelocity=2, worldAlpha=1}
   }
   n9 = RigidBody(ComponentPose=@n8) {
    bClampLinearTranslationLimitToRefPose=true
    bOverrideWorldGravity=true
    overridePhysicsAsset=&/M_MED_FluidWedge_Physics_RLeg
    simulationSpace=BaseBoneSpace
    baseBoneRef {boneName=thigh_r}
    componentAppliedLinearAccClamp=<250 250 250>
    overrideWorldGravity {z=-980}
    simSpaceSettings {maxAngularAcceleration=150, maxAngularVelocity=150, maxLinearAcceleration=2, maxLinearVelocity=2, worldAlpha=1}
   }
   n10 = "Constraint - dyn_fastener_thigh_r (T)"(ComponentPose=@n9, Alpha=1, ConstraintWeights_0=1) {
    boneToModify {boneName=dyn_fastener_thigh_r}
    constraintSetup [{transformType=Translation, targetBone {boneName=fastener_thigh_r}}]
    constraintWeights [1]
   }
   n11 = "Constraint - dyn_fastener_calf_r (T)"(ComponentPose=@n10, Alpha=1, ConstraintWeights_0=1) {
    boneToModify {boneName=dyn_fastener_calf_r}
    constraintSetup [{transformType=Translation, targetBone {boneName=fastener_calf_r}}]
    constraintWeights [1]
   }
   n12 = "Constraint - dyn_fastener_foot_r (T)"(ComponentPose=@n11, Alpha=1, ConstraintWeights_0=1) {
    boneToModify {boneName=dyn_fastener_foot_r}
    constraintSetup [{transformType=Translation, targetBone {boneName=fastener_foot_r}}]
    constraintWeights [1]
   }
   n13 = "Constraint - dyn_fastener_thigh_l (T)"(ComponentPose=@n12, Alpha=1, ConstraintWeights_0=1) {
    boneToModify {boneName=dyn_fastener_thigh_l}
    constraintSetup [{transformType=Translation, targetBone {boneName=fastener_thigh_l}}]
    constraintWeights [1]
   }
   n14 = "Constraint - dyn_fastener_calf_l (T)"(ComponentPose=@n13, Alpha=1, ConstraintWeights_0=1) {
    boneToModify {boneName=dyn_fastener_calf_l}
    constraintSetup [{transformType=Translation, targetBone {boneName=fastener_calf_l}}]
    constraintWeights [1]
   }
   n15 = "Constraint - dyn_fastener_foot_l (T)"(ComponentPose=@n14, Alpha=1, ConstraintWeights_0=1) {
    boneToModify {boneName=dyn_fastener_foot_l}
    constraintSetup [{transformType=Translation, targetBone {boneName=fastener_foot_l}}]
    constraintWeights [1]
   }
   Output Pose(Result=Component To Local(ComponentPose=@n15)) {layerGroup=None}
   n16 = Control Rig(Source=Copy Pose From Mesh() {bUseAttachedParent=true}) {controlRigAssetReference {controlRigAsset=&/M_MED_FluidWedge_CtrlRig}}
   Save cached pose 'PreSim'(Pose=Component To Local(ComponentPose=Local To Component(LocalPose=@n16))) {cacheName=PreSim}
  ]}
 EventGraph {nodes [Event BlueprintUpdateAnimation()]}
]
variables []
